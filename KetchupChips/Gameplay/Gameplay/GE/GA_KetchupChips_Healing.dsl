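; GA_KetchupChips_Healing is a gameplay ability Blueprint, apparently for healing from the KetchupChips consumable.
; Activation simply calls the parent ability's logic, so the healing behavior lives in the parent class.
; The end-ability event is present but empty.
; A note records a GameplayCue.Shield.PotionConsumed tag for possible reuse with shields, but nothing here uses it.
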(event Ability|EventActivateAbility
  ; Tag in case you need to bring it back for Shields: (TagName="GameplayCue.Shield.PotionConsumed")
  (|CallParentFunction))

(event Ability|EventOnEndAbility
  (params (bWasCancelled :type Boolean)))
